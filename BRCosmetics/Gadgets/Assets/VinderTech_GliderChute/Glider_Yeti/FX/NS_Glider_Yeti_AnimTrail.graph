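# Niagara system for the Yeti glider's animation trails, drawing two pale blue, brightened ribbon trails from bone or socket pairs on a skeletal mesh.
# Shared FN_ AnimTrail modules spawn, initialize and update the particles.
# In gameplay, time-smoothed User.ForwardDot, clamped, scales an alpha curve that fades over particle life.
# In the front end, that input is fixed at one and particles gain a constant velocity, which is zeroed in gameplay.
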NiagaraSystem /BRCosmetics/Gadgets/Assets/VinderTech_GliderChute/Glider_Yeti/FX/NS_Glider_Yeti_AnimTrail
dependencies {
 usedDataInterfaces [&NiagaraDataInterfaceCurve, &NiagaraDataInterfaceSkeletalMesh]
 usedDynamicInputs []
 usedModules [
  &self:SystemSpawnScript.SystemScriptSource.SystemScriptGraph.NiagaraNodeAssignment_0.SetVariables_#1
  &SystemState
 ]
}
emitters {AnimTrail {
  emitterUpdateScript [
   &EmitterState(Life Cycle Mode=::NewEnumerator1, Loop Behavior=::NewEnumerator1, Loop Duration Mode=::NewEnumerator1, UseLoopDelay=false)
   &FN_SpawnAnimTrail(Number of Anim Trails=2)
  ]
  particleSpawnScript [
   &FN_InitializeAnimTrailBoneIndecies()
   &FN_InitializeAnimTrailBoneIndecies(AnimTrail Selection=1, First Bone or Socket Index=2, Second Bone or Socket Index=3) {module=FN_InitializeAnimTrailBoneIndecies001}
   &FN_InitializeAnimTrail(Color=<rgba 0.3712 0.552 1 1>, SkeletalMesh=$System.Skeletal Mesh)
   &AddVelocity(Velocity=<0 1000 0>, Velocity Speed Scale=&FN_FrontEndFloatModifier(Gameplay=0, Front End=1))
  ]
  particleUpdateScript [
   &InterpolateOverTime(
    Rate Of Change=3
    Target Float Value:NiagaraFloat=&FN_FrontEndFloatModifier(Gameplay=$User.ForwardDot, Front End=1)
   )
   &ParticleState()
   &FN_UpdateAnimTrail(SkeletalMesh=$System.Skeletal Mesh)
   &ScaleColor(
    Scale Alpha=&FloatFromCurve(
     FloatCurve=DataInterface() {propertyValues {Curve {keys [<at 0 1>,<at 1 0>]}}}
     Scale Curve=&Multiply_Float(A=2, B=&ClampFloat(Float=$Particles.InterpolateOverTime.InterpolateOverTime_Float.Moving Average))
    )
    Scale RGB=&Multiply_VectorByFloat(Vector=<1 1 1>, Float=8)
    ScaleA=true
    ScaleRGB=true
   )
   &SolveForcesAndVelocity(Clamp Velocity=false, Limit Acceleration=false)
  ]
  rendererClasses [&NiagaraRibbonRendererProperties]
  renderers [{rendererClass=&NiagaraRibbonRendererProperties}]
 }}
summary {
 systemName=NS_Glider_Yeti_AnimTrail
 emitters [{emitterName=AnimTrail, rendererClasses [&NiagaraRibbonRendererProperties]}]
 userVariables [
  User.bIsFrontEnd {type=NiagaraBool}
  User.bIsFrontEndPreview {type=NiagaraBool}
  User.ForwardDot {type=NiagaraFloat, defaultValue=0, description=""}
  User.RibbonIDOffset {type=NiagaraInt32}
 ]
}
